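# Translucent, unlit, two-sided master material for dev-test FX rig trails, enabled for skeletal meshes and mesh deformers.
# It uses no textures; SphereMasks on the V texture coordinate act as one-dimensional gradients.
# One mask blends two color parameters into a ramp, and a second forms an opacity fade.
# Scalar parameters expose the start, size and softness of each gradient.
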Material /BRCosmetics/Animation/Props/DevTest_FXRig/Meshes/M_DevTest_Trail_Master
graph {
 properties {
  bAutomaticallySetUsageInEditor=false
  bUsedWithMeshDeformer=true
  bUsedWithSkeletalMesh=true
  blendMode=BLEND_Translucent
  shadingModel=MSM_Unlit
  twoSided=true
  usedShadingModels=Unlit
 }
 nodes [
  n1 = TextureCoordinate()
  LinearInterpolate(
   A=VectorParameter() {parameterName=Color B, defaultValue=<rgba 0.5298 1 0 1>}.RGB
   B=VectorParameter() {parameterName=Color A, defaultValue=<rgba 1 0 0.0369 1>}.RGB
   Alpha=SphereMask(
    A=ComponentMask(None=@n1) {g=true}
    B=ScalarParameter() {parameterName=Color Ramp Start}
    Radius=ScalarParameter() {defaultValue=0.5, parameterName=Color Ramp Gradient}
    Hardness=ScalarParameter() {defaultValue=0.5, parameterName=Color Ramp Hardness}
   ) {hardnessPercent=0}
  )
  SphereMask(
   A=ComponentMask(None=TextureCoordinate()) {g=true}
   B=ScalarParameter() {parameterName=Opacity Fade Start}
   Radius=ScalarParameter() {defaultValue=0.626, parameterName=Opacity Fade Size}
   Hardness=ScalarParameter() {defaultValue=0.5, parameterName=Opacity Fade Gradient}
  ) {hardnessPercent=0}
 ]
}
